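# Animation Blueprint for the robot hologram NPC in the BattlePassS23 hologram device content, showing a static-named T-pose sequence that spins continuously.
# Each animation update subtracts a fixed step from a yaw value, ignoring delta time, so spin speed follows update rate.
# The AnimGraph applies that rotation additively to the root bone in component space over the sequence pose.
# Copy Pose From Mesh and DefaultSlot nodes are present but unconnected.
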AnimBlueprint /BattlePassS23/QuestActors/Narrative/HologramDevice/Anim/M_MED_Robot_Hologram_Pose_AnimBP
target_skeleton=/BRCosmetics/Characters/Player/Male/Medium/Bodies/M_MED_Robot_Hologram_NPC/Meshes/M_MED_Robot_Hologram_NPC_Skeleton
graphs [
 AnimGraph {nodes [
   Copy Pose From Mesh() {bUseAttachedParent=true}
   n1 = Local To Component(LocalPose=Sequence Player 'TPOSE_NPC_RIG'() {sequence=&/TPOSE_NPC_RIG})
   n2 = Get RotVar() -> (RotVar:Rotator)
   Output Pose(Result=Component To Local(ComponentPose="Transform (Modify) Bone - Bone: root"(ComponentPose=@n1, Translation="0,0,0", Rotation=@n2) {rotationMode=BMM_Additive, alphaBoolBlend {blendOption=Linear}, boneToModify {boneName=root}})) {layerGroup=None}
   Slot 'DefaultSlot'()
  ]}
 EventGraph {nodes [
   Event BlueprintUpdateAnimation()
   > n1 = Set Rotation(Rotation:double=float - float(A=Get Rotation() -> (Rotation:double), B=0.4)) -> (Output_Get:double)
   > Set RotVar(RotVar:Rotator=MakeRotator(Yaw=@n1)) -> (Output_Get:Rotator)
   TryGetPawnOwner()
  ]}
]
variables [
 f {type=number}
 Rotation {type=number}
 RotVar {type=Rotator}
]
