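; Human variant of the Season 23 battle pass hologram device quest actor, which projects an animated mannequin after player interaction.
; On begin play it caches the mannequin from a child actor component.
; After interaction it has that mannequin load and play the configured montage, if valid.
; A timeline then fades the hologram in by lerping cone and mannequin root world scales from initial values up to their captured current scales.
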(event EventBeginPlay
  (|CallParentFunction)
  (bind _asbp_s23_hologram_device_mannequin (Utilities|Casting|CastToBP_S23_HologramDevice_Mannequin (Variables|/Script/Engine.ChildActorComponent|GetChildActor (Variables|Default|GetMannequin))))
  (Variables|Default|SetMannequinActor _asbp_s23_hologram_device_mannequin))

(event VisualState|Meshes|EventPostInteractionVisuals
  (params (PlayerController :type Fort%20Player%20Controller%20Object%20Reference))
  (bind _mannequinactor (Variables|Default|GetMannequinActor))
  (|CallParentFunction PlayerController)
  (Utilities|FlowControl|Sequence
    (:then_0
      (Utilities|IsValid _mannequinactor
        (:Is%20Valid
          (Class|BPS23HologramDeviceMannequin|LoadandPlayCurrentMontage _mannequinactor (Variables|Default|GetAnimation%20Montage) 0.0 false))))
    (:then_1
      ;[1 1 1] Fade-in added by Bruno Afonseca
      (Variables|Default|SetPostInteractionVisualConeFinalScale (Transformation|GetWorldScale (Variables|Default|GetFX%20Holo_Cone)))
      (Variables|Default|SetPostInteractionVisualDeviceFinalScale (Transformation|GetWorldScale (Variables|/Script/Engine.Actor|GetRootComponent (Variables|Default|GetMannequinActor))))
      (bind (_direction _scalelerpalpha) (Timeline|PostInteractionVisualFadeIn PlayFromStart)
        (:Update
          (Transformation|SetWorldScale3D (Variables|Default|GetFX%20Holo_Cone) (Math|Vector|Lerp(Vector) (Variables|Default|GetPostInteractionVisualConeInitialScale) (Variables|Default|GetPostInteractionVisualConeFinalScale) _scalelerpalpha))
          (Transformation|SetWorldScale3D (Variables|/Script/Engine.Actor|GetRootComponent (Variables|Default|GetMannequinActor)) (Math|Vector|Lerp(Vector) (Variables|Default|GetPostInteractionVisualDeviceInitialScale) (Variables|Default|GetPostInteractionVisualDeviceFinalScale) _scalelerpalpha)))))))

(event Collision|EventActorBeginOverlap
  (params (OtherActor :type Actor%20Object%20Reference)))

(event EventTick
  (params (DeltaSeconds :type Float%20%28single-precision%29)))

(fn ConstructionScript
  (|CallParentFunction))
