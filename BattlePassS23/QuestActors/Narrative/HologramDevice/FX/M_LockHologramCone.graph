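# Translucent, two-sided unlit material for the light cone of a quest hologram lock device, usable on Niagara mesh particles.
# Spawn Animation Time drives per-axis vertex offsets that collapse or expand the cone, with vertex-red-phased idle wobble and a looping debug switch.
# Opacity sums a UV gradient, panned noise rays and cloud smoke, shaped by Fresnel, depth fade and optional mesh-normal fade.
# Squared cone color feeds emissive, optionally tinted by particle color.
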Material /BattlePassS23/QuestActors/Narrative/HologramDevice/FX/M_LockHologramCone
graph {
 properties {
  bUsedWithNiagaraMeshParticles=true
  blendMode=BLEND_Translucent
  pixelDepthOffsetMode=PDOM_Legacy
  shadingModel=MSM_Unlit
  twoSided=true
  usedShadingModels=Unlit
 }
 nodes [
  n1 = Cosine(None=Multiply(A=Time()) {constB=0.3})
  n2 = StaticSwitchParameter(
   True=Saturate(None=Add(A=@n1) {constB=0.5})
   False=ScalarParameter() {defaultValue=1, group=Logic, parameterName=Spawn Animation Time, sliderMax=1}
  ) {group=0_DEBUG, parameterName=DEBUG MODE}
  n3 = Subtract(A=@n2) {constB=0.95}
  n4 = Multiply(A=VertexColor().R) {constB=0.5}
  n5 = Time()
  n6 = Multiply(
   A=Saturate(None=Sign(None=@n3))
   B=Multiply(A=Add(A=@n4, B=@n5), B=ScalarParameter() {defaultValue=0.4, group=Animation, parameterName=Idle Animation Speed})
  )
  n7 = Add(
   A=OneMinus(None=@n2)
   B=Multiply(A=Cosine(None=@n6), B=ScalarParameter() {defaultValue=0.1, group=Animation, parameterName=Idle Animation Magnitude}) {constB=0.1}
  )
  n8 = TransformPosition(None=WorldPosition().XYZ) {transformSourceType=TRANSFORMPOSSOURCE_World}
  n9 = ComponentMask(None=@n8) {b=true}
  n10 = Transform(None=Constant3Vector() {constant=<rgba 0 0 1 0>}) {transformSourceType=TRANSFORMSOURCE_Local}
  n11 = Multiply(A=@n2) {constB=5}
  n12 = Multiply(A=OneMinus(None=Saturate(None=@n11))) {constB=-1}
  n13 = Multiply(
   A=ComponentMask(None=@n8) {r=true}
   B=Transform(None=Constant3Vector() {constant=<rgba 1 0 0 0>}) {transformSourceType=TRANSFORMSOURCE_Local}
  )
  n14 = Multiply(
   A=ComponentMask(None=@n8) {g=true}
   B=Transform(None=Constant3Vector() {constant=<rgba 0 1 0 0>}) {transformSourceType=TRANSFORMSOURCE_Local}
  )
  Add(A=Multiply(A=Multiply(A=@n7) {constB=-1}, B=Multiply(A=@n9, B=@n10)), B=Add(A=Multiply(A=@n12, B=@n13), B=Multiply(A=@n12, B=@n14)))
  n15 = VectorParameter() {group=Color, parameterName=Cone Color, defaultValue=<rgba 0.0437 0.4735 1 3>}
  n16 = Multiply(A=@n15.RGB, B=@n15.RGB)
  n17 = ParticleColor()
  n18 = Multiply(A=@n16, B=@n17.RGB)
  n19 = ScalarParameter() {parameterName=Use_Particle_Color}
  MaterialFunctionCall(Emissive=Desaturation(None=LinearInterpolate(A=@n16, B=@n18, Alpha=@n19), Fraction=ScalarParameter() {parameterName=FinalSaturation})) {materialFunction=&MF_FX_EmissiveAdjustments}
  n20 = ComponentMask(None=TextureCoordinate()) {g=true}
  n21 = Power(Base=@n20, Exp=ScalarParameter() {defaultValue=6, group=Opacity, parameterName=Cone Exponent}) {constExponent=10}
  n22 = Panner(Coordinate=TextureCoordinate() {uTiling=2, vTiling=0}) {speedX=0.05, speedY=0.1}
  n23 = Multiply(
   A=LinearInterpolate(
    A=ScalarParameter() {defaultValue=-1, group=Opacity, parameterName=Opacity Remap}
    Alpha=TextureSample(UVs=@n22) {texture=&T_NoisePack_01}.R
   ) {constA=-1.5}
   B=Power(Base=@n20, Exp=ScalarParameter() {defaultValue=5, group=Opacity, parameterName=Ray Exponent}) {constExponent=5}
  )
  n24 = TextureSample(UVs=Panner() {speedX=0.1, speedY=0.2}) {texture=/Game/Packages/Fortnite_UniqueMaterials_1/PostProcess/T_TilingClouds_02}
  n25 = Multiply(A=Multiply(A=@n20, B=@n20), B=Add(A=LinearInterpolate(Alpha=@n24.R) {constA=-0.3, constB=0.1}))
  n26 = ScalarParameter() {defaultValue=0.2, group=Opacity, parameterName=Smoke Strength}
  n27 = OneMinus(None=Fresnel(ExponentIn=ScalarParameter() {defaultValue=1, group=Opacity, parameterName=Fresnel Exponent}) {baseReflectFraction=0, exponent=3})
  n28 = DepthFade(FadeDistance=ScalarParameter() {defaultValue=5, group=Opacity, parameterName=DepthFadeDistance}) {fadeDistanceDefault=5}
  n29 = TextureCoordinate()
  n30 = Sine(None=Multiply(A=ComponentMask(None=@n29) {g=true}) {constB=0.5})
  n31 = Multiply(A=Add(A=Add(A=@n21, B=@n23), B=Multiply(A=@n25, B=@n26)), B=Multiply(A=Multiply(A=@n27, B=@n28), B=Multiply(A=@n30, B=@n30)))
  n32 = PixelNormalWS()
  n33 = VertexNormalWS()
  n34 = QualitySwitch(Default=@n32, Low=@n33, High=@n32, Medium=@n33, Epic=@n32)
  n35 = Abs(None=DotProduct(A=CameraVectorWS(), B=ShadingPathSwitch(Default=@n34, Deferred=@n34, Forward=@n33, Mobile=@n33)))
  n36 = VectorParameter() {
   group=Base
   parameterName=Mesh Fresnel Fade
   sortPriority=1
   channelNames {a="", b=Power, g=Invert, r=Use Mesh Fresnel Fade}
   defaultValue=<rgba 0 0 1 0>
  }
  n37 = Power(Base=LinearInterpolate(A=@n35, B=OneMinus(None=@n35), Alpha=@n36.Invert), Exp=@n36.Power)
  n38 = LinearInterpolate(A=@n31, B=Multiply(A=Saturate(None=@n37), B=@n31), Alpha=@n36.Use Mesh Fresnel Fade)
  Multiply(A=LinearInterpolate(A=@n31, B=Multiply(A=@n17.A, B=@n38), Alpha=@n19)) {constB=0.6}
 ]
}
